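# Niagara system for the Enchant emote's heart effect, a one-burst heart pop with shockwave and flare.
# Em_Heart pairs sprite and mesh renderers, lasting the system loop, with a curve that grows, holds and shrinks the mesh.
# Em_Shock adds a brief expanding pink mesh that fades out.
# Flare_With_Streak drives separate flash size and color curves, shrinking in front end via FN_FrontEndFloatModifier; the Em_Sparks emitter is disabled.
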NiagaraSystem /BRCosmetics/Animation/Game/MainPlayer/Emotes/Enchant/FX/NS_Enchant_Heart
dependencies {
 usedDataInterfaces [&NiagaraDataInterfaceCurve, &NiagaraDataInterfaceCamera, &NiagaraDataInterfaceArrayColor]
 usedDynamicInputs []
 usedModules [&SystemState]
}
emitters {
 Em_Heart {
  emitterUpdateScript [
   &EmitterState()
   &SpawnBurst_Instantaneous(Use Spawn Probability=false, UseLoopCountLimit=false)
  ]
  particleSpawnScript [
   &InitializeParticle(
    Color=<rgba 1 1 1 0.75>
    Lifetime=$System.CurrentLoopDuration
    Mesh Scale Mode=::NewEnumerator3
    Mesh Uniform Scale=0.4
    Position Offset=<0 0 -24.6>
    Sprite Size Mode=::NewEnumerator3
    Uniform Sprite Size=150
    UsePositionOffset=false
    Write Material Random=true
   )
   &InitializeParticle(
    Color=<rgba 1 0 0 0.75>
    Lifetime=0.667
    Mesh Scale Mode=::NewEnumerator3
    Mesh Uniform Scale=1.1
    Position Offset=<5 0 -24>
    Sprite Size Mode=::NewEnumerator3
    Uniform Sprite Size=150
    UsePositionOffset=true
    Write Material Random=true
   ) {enabled=false, module=InitializeParticle001}
   &InitialMeshOrientation(
    Mesh Orientation Mode=::NewEnumerator7
    Rotation=&MakeVector(
     X=&AngleConversion(Angle:NiagaraFloat=0, Angle Input:ENiagara_AngleInput=::NewEnumerator0, Angle Output:ENiagara_AngleInput=::NewEnumerator1)
     Y=&AngleConversion(Angle:NiagaraFloat=0, Angle Input:ENiagara_AngleInput=::NewEnumerator0, Angle Output:ENiagara_AngleInput=::NewEnumerator1)
     Z=&AngleConversion(Angle:NiagaraFloat=0, Angle Input:ENiagara_AngleInput=::NewEnumerator0, Angle Output:ENiagara_AngleInput=::NewEnumerator1)
    )
    Use Rotation Vector=true
   )
   &InitialMeshOrientation(
    Mesh Orientation Mode=::NewEnumerator7
    Rotation=&MakeVector(
     X=&AngleConversion(Angle:NiagaraFloat=0, Angle Input:ENiagara_AngleInput=::NewEnumerator0, Angle Output:ENiagara_AngleInput=::NewEnumerator1)
     Y=&AngleConversion(Angle:NiagaraFloat=-14.088, Angle Input:ENiagara_AngleInput=::NewEnumerator0, Angle Output:ENiagara_AngleInput=::NewEnumerator1)
     Z=&AngleConversion(Angle:NiagaraFloat=0, Angle Input:ENiagara_AngleInput=::NewEnumerator0, Angle Output:ENiagara_AngleInput=::NewEnumerator1)
    )
    Use Rotation Vector=true
   ) {enabled=false, module=InitialMeshOrientation001}
   &CameraOffset(Camera Offset Amount=-25)
   /SetVariables() {isSetParameters=true}
  ]
  particleUpdateScript [
   &ParticleState()
   &ScaleSpriteSize(Uniform Curve Sprite Scale=DataInterface() {propertyValues {Curve {keys [<at 0 0 RCIM_Cubic>, {arriveTangentWeight=0.05, interpMode=RCIM_Cubic, tangentMode=RCTM_User, time=0.15, value=1}]}}}) {enabled=false}
   &ScaleMeshSize(Scale Factor=&VectorFromFloat(Value=&FloatFromCurve(FloatCurve=DataInterface() {propertyValues {Curve {keys [
          {tangentMode=RCTM_User, time=0, value=0}
          {arriveTangent=2.5714, interpMode=RCIM_Cubic, leaveTangent=2.5714, time=0.2, value=1.8}
          {arriveTangent=-2.6431, interpMode=RCIM_Cubic, leaveTangent=-2.6431, tangentMode=RCTM_User, time=0.7, value=1.8}
          <at 1 0>
         ]}}})))
   /SetVariables(Particles.ComponentPosition:NiagaraPosition=&AddVectorToPosition(
     Position=$Particles.Position
     Vector=&Multiply_VectorByFloat(
      Vector=&ReturnCameraVector(Camera Vectors=::NewEnumerator4, Coordinate Space=::Simulation, Origin Position=$Particles.Position)
      Float=2
     )
    )) {enabled=false, isSetParameters=true}
   /SetVariables(Particles.DisplayPosition:NiagaraPosition=<-2 0 2>) {isSetParameters=true}
   &ScaleColor(ScaleA=true, ScaleRGB=true)
  ]
  rendererClasses [&NiagaraSpriteRendererProperties, &NiagaraMeshRendererProperties]
  renderers [
   {rendererClass=&NiagaraSpriteRendererProperties}
   {rendererClass=&NiagaraMeshRendererProperties, rendererIndex=1}
  ]
 }
 Em_Shock {
  emitterUpdateScript [
   &EmitterState()
   &SpawnBurst_Instantaneous(Use Spawn Probability=false, UseLoopCountLimit=false)
  ]
  particleSpawnScript [&InitializeParticle(
    Color=<rgba 4 0.5319 2.2081 3>
    Lifetime=0.2
    Mesh Scale Mode=::NewEnumerator3
    Mesh Uniform Scale=0.3
    UsePositionOffset=false
    Write Material Random=true
   )]
  particleUpdateScript [
   &ParticleState()
   &ScaleColor(
    Scale Alpha=&FloatFromCurve(FloatCurve=DataInterface() {propertyValues {Curve {keys [<at 0 1>, {interpMode=RCIM_Cubic, tangentMode=RCTM_User, time=0.1, value=1}, {interpMode=RCIM_Cubic, tangentMode=RCTM_User, time=1, value=0}]}}})
    ScaleA=true
    ScaleRGB=true
   )
   &DynamicMaterialParameters(
    Index 0 Param 1=1
    Index 0 Param 2=-0.5
    Index 0 Param 3=1
    Index 0 Param 4=1
    Module.Param0WriteEnabled=true
    Module.Param1WriteEnabled=true
    Module.Param2WriteEnabled=true
    Module.Param3WriteEnabled=true
   )
   &ScaleMeshSize(Scale Factor=&VectorFromFloat(Value=&FloatFromCurve(FloatCurve=DataInterface() {propertyValues {Curve {keys [
          {arriveTangent=4.1785, interpMode=RCIM_Cubic, leaveTangent=4.1785, tangentMode=RCTM_User, time=0, value=0.05}
          {arriveTangent=0.9995, interpMode=RCIM_Cubic, leaveTangent=0.9995, tangentMode=RCTM_User, time=0.3, value=0.6}
          {arriveTangent=-0.0405, interpMode=RCIM_Cubic, leaveTangent=-0.0405, tangentMode=RCTM_User, time=1, value=1}
         ]}}})))
  ]
  rendererClasses [&NiagaraMeshRendererProperties]
  renderers [{rendererClass=&NiagaraMeshRendererProperties, rendererIndex=0}]
 }
 Em_Sparks {
  bEnabled=false
  emitterUpdateScript [
   &EmitterState(Life Cycle Mode=::NewEnumerator1, Loop Behavior=::NewEnumerator1, Loop Delay=0.1, Loop Duration=1.5, UseLoopDelay=false)
   &SpawnBurst_Instantaneous(Spawn Count=22, Use Spawn Probability=false)
  ]
  particleSpawnScript [
   &InitializeParticle(
    Color=&SelectLinearColorFromArray(
     Array Sampling Mode=::NewEnumerator0
     Direct Array Index=?
     Color Selection Array=DataInterface() {propertyValues {ColorData [<rgba 1 0 0.0304 1>, <rgba 1 0 0.6649 1>]}}
     Random Seed=0
    )
    Lifetime=0.4
    Sprite Size Max=<1.5 3>
    Sprite Size Min=<1 2>
    Sprite Size Mode=::NewEnumerator2
    UsePositionOffset=false
    Write Material Random=true
   )
   &SphereLocation(
    MaskSpawnGroup=false
    Override Seed=false
    Sphere Radius=30
    Surface Only=false
    UseNonUniformScale=true
    UseOffset=false
    UseSphereOrientation=true
   )
   &AddVelocityFromPoint(Module.Use Falloff Distance=false, Velocity Strength=&RandomRangeFloat(Minimum=5, Maximum=80))
   &InheritVelocity(
    Inherited Velocity Amount Scale=&VectorFromFloat(Value=1)
    Inherited Velocity Speed Limit=50
    Speed Limit=true
    UseSpeedThreshold=false
   )
  ]
  particleUpdateScript [
   &ParticleState()
   &ScaleColor(
    Scale Alpha=&FNTimeOfDayScalar(Night Scale=0.6, Morning Scale=0.4, Evening Scale=0.4)
    Scale RGB=&VectorFromFloat(Value=&FloatFromCurve(FloatCurve=DataInterface() {propertyValues {Curve {keys [<at 0 10 RCIM_Cubic>,<at 0.9 1 RCIM_Cubic>]}}}))
    ScaleA=true
    ScaleRGB=true
   ) {module=ScaleColor001}
   &Drag(Drag=&RandomRangeFloat(Minimum=3, Maximum=6), Use Linear Drag=true, Use Rotational Drag=false)
   &SolveForcesAndVelocity(Clamp Velocity=false, Limit Acceleration=false)
   &ScaleSpriteSize(Scale Factor=&MakeVector2D(
     X=&FloatFromCurve(FloatCurve=DataInterface() {propertyValues {Curve {keys [<at 0 0>,<at 0.1 1>,<at 1 0.5>]}}})
     Y=&FloatFromCurve(FloatCurve=DataInterface() {propertyValues {Curve {keys [<at 0 0.5>,<at 0.1 1>,<at 1 0>]}}})
    ))
   &ScaleSpriteSizeBySpeed(Max Scale Factor=<0.2 5>, Min Scale Factor=<1 2>)
   &CameraOffset(Camera Offset Amount=-15)
  ]
  rendererClasses [&NiagaraSpriteRendererProperties]
  renderers [{rendererClass=&NiagaraSpriteRendererProperties}]
 }
 Flare_With_Streak {
  emitterUpdateScript [
   &EmitterState(Life Cycle Mode=::NewEnumerator1, Loop Behavior=::NewEnumerator1, Loop Duration=1, UseLoopDelay=false)
   &SpawnBurst_Instantaneous(Spawn Count=1, Use Spawn Probability=false)
  ]
  particleSpawnScript [
   &InitializeParticle(
    Color=<rgba 0 1.0667 2 3>
    Lifetime=0.4
    Sprite Rotation Mode=::NewEnumerator2
    Sprite Size Mode=::NewEnumerator3
    Uniform Sprite Size=100
    UsePositionOffset=false
    Write Material Random=true
   )
   &DynamicMaterialParameters(
    Index 0 Param 1=&RandomRangeFloat(Minimum=0.4, Maximum=0.6)
    Index 0 Param 2=1
    Index 0 Param 3=1
    Index 0 Param 4=1
    Module.Param0WriteEnabled=true
    Module.Param1WriteEnabled=true
    Module.Param2WriteEnabled=true
    Module.Param3WriteEnabled=true
   )
   &SystemLocation(MaskSpawnGroup=false)
   /SetVariables(Particles.StreakRotation=0.2) {isSetParameters=true}
   &CameraOffset(Camera Offset Amount=-1)
  ]
  particleUpdateScript [
   &ParticleState()
   &SolveForcesAndVelocity(Clamp Velocity=false, Limit Acceleration=false)
   &ScaleColor(
    Scale Alpha=&Multiply_Float(
     A=1
     B=&FloatFromCurve(FloatCurve=DataInterface() {propertyValues {Curve {keys [<at 0 0>,<at 0.2 1>,<at 0.8 1>,<at 1 0>]}}})
    )
    ScaleA=true
    ScaleRGB=true
   )
   &ScaleSpriteSize(Scale Factor=&Vector2DFromFloat(Value=&FloatFromCurve(
      FloatCurve=DataInterface() {propertyValues {Curve {keys [<at 0 1.5 RCIM_Cubic>,<at 1 0>]}}}
      Scale Curve=&FN_FrontEndFloatModifier(Gameplay=1, Front End=0.7)
     )))
   /SetVariables(Particles.Flash size:Vector2f=&Multiply_Vector2D_ByFloat(
     Float=&Multiply_Float(
      A=400
      B=&FloatFromCurve(
       FloatCurve=DataInterface() {propertyValues {Curve {keys [
           {arriveTangent=4.1785, interpMode=RCIM_Cubic, leaveTangent=4.1785, tangentMode=RCTM_User, time=0, value=0}
           {arriveTangent=0.9995, interpMode=RCIM_Cubic, leaveTangent=0.9995, tangentMode=RCTM_User, time=0.1378, value=0.6825}
           {arriveTangent=-0.0405, interpMode=RCIM_Cubic, leaveTangent=-0.0405, tangentMode=RCTM_User, time=0.62, value=1}
          ]}}}
       Scale Curve=&FN_FrontEndFloatModifier(Gameplay=0.9, Front End=0.7)
      )
     )
     Vector2D=&Multiply_Vector2D_ByFloat(Float=0.35, Vector2D=<1.5 1>)
    )) {isSetParameters=true}
   /SetVariables(Particles.Flash color:LinearColor=&MakeLinearColorFromVectorAndFloat(
     "Vector (RGB)"=&MakeVectorFromLinearColorRGB(Color=<rgba 0.3302 1 0.8326 1>)
     "Float (Alpha)"=&Multiply_Float(
      A=&Multiply_Float(
       A=1
       B=&FloatFromCurve(FloatCurve=DataInterface() {propertyValues {Curve {keys [<at 0 0>,<at 0.124 1>,<at 0.496 1>,<at 0.62 0>]}}})
      )
      B=1
     )
    )) {isSetParameters=true}
  ]
  rendererClasses [&NiagaraSpriteRendererProperties]
  renderers [
   {rendererClass=&NiagaraSpriteRendererProperties}
   {rendererClass=&NiagaraSpriteRendererProperties, rendererIndex=1}
  ]
 }
}
summary {
 systemName=NS_Enchant_Heart
 emitters [
  {emitterName=Em_Heart, rendererClasses [&NiagaraSpriteRendererProperties, &NiagaraMeshRendererProperties]}
  {bEnabled=false, emitterName=Em_Sparks, rendererClasses [&NiagaraSpriteRendererProperties]}
  {emitterName=Em_Shock, rendererClasses [&NiagaraMeshRendererProperties]}
  {emitterName=Flare_With_Streak, rendererClasses [&NiagaraSpriteRendererProperties]}
 ]
 userVariables [
  User.bIsFrontEnd {type=NiagaraBool}
  User.bIsFrontEndPreview {type=NiagaraBool}
 ]
}
